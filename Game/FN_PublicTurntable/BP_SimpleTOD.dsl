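; BP_SimpleTOD is a lighting rig actor that sets directional light, skylight and exponential height fog from its own variables.
; The Construction Script runs a three-step Sequence, one step per component.
; It copies intensity, color, shadow and scattering values to the sun light, color, cubemap, intensity and occlusion to the skylight, and density, colors, opacity and scattering to the fog.
; BeginPlay, Overlap and Tick are empty, so nothing here animates time of day.
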(event EventBeginPlay)

(event Collision|EventActorBeginOverlap
  (params (OtherActor :type Actor%20Object%20Reference)))

(event EventTick
  (params (DeltaSeconds :type Float%20%28single-precision%29)))

(fn ConstructionScript
  (bind _directionallight (Variables|Default|GetDirectionalLight))
  (bind _skylight (Variables|Default|GetSkyLight))
  (bind _exponentialheightfog (Variables|Default|GetExponentialHeightFog))
  (Utilities|FlowControl|Sequence
    (:then_0
      ;[1 1 1] Directional Light Settings
      (Rendering|Components|Light|SetIntensity _directionallight (Variables|DirectionalLight|GetDirectional%20Light%20Intensity))
      (Rendering|Components|Light|SetLightColor _directionallight (Variables|DirectionalLight|GetDirectional%20Light%20Color) true)
      (Rendering|Components|Light|SetVolumetricScatteringIntensity _directionallight (Variables|DirectionalLight|GetDirectional%20Light%20Volumetric%20Scattering%20Intensity) :!class /Script/Engine.LightComponent)
      (Rendering|Components|Light|SetCastShadows (Variables|Default|GetDirectionalLight) (Variables|DirectionalLight|GetCast%20Shadow)))
    (:then_1
      ;[1 1 1] Skylight Settings
      (Rendering|Components|SkyLight|SetLightColor _skylight (Variables|Skylight|GetSkylight%20Color))
      (SkyLight|SetCubemap _skylight (Variables|Skylight|GetSkylight%20Image))
      (Rendering|Components|SkyLight|SetIntensity _skylight (Variables|Skylight|GetSkylight%20intensity))
      (Rendering|Components|SkyLight|SetMinOcclusion _skylight (Variables|Skylight|GetSky%20Min%20Occlusion)))
    (:then_2
      ;[1 1 1] Height Fog Settings
      (Rendering|Components|ExponentialHeightFog|SetFogDensity _exponentialheightfog (Variables|Fog|GetFog%20Density))
      (Rendering|Components|ExponentialHeightFog|SetDirectionalInscatteringColor _exponentialheightfog (Variables|Fog|GetDirectional%20Inscattering%20Color))
      (Rendering|Components|ExponentialHeightFog|SetFogInscatteringColor _exponentialheightfog (Variables|Fog|GetFog%20Inscattering%20Color))
      (Rendering|Components|ExponentialHeightFog|SetFogMaxOpacity _exponentialheightfog (Variables|Fog|GetFog%20Max%20Opacity))
      (Rendering|VolumetricFog|SetVolumetricFogScatteringDistribution _exponentialheightfog (Variables|Fog|GetScattering%20Distrubtion)))))
