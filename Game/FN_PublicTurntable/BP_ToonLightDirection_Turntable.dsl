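; Turntable actor Blueprint that supplies the toon lighting sun direction to materials from the orientation of its Arrow component.
; It negates the arrow's forward vector and writes it as SunlightDirectionalVector into the shared FortniteMaterialParameters collection.
; The write runs in the Construction Script and again on BeginPlay, with no per-frame update.
; The Tick and overlap events are empty.
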(event EventBeginPlay
  (Rendering|Material|SetVectorParameterValue "/Game/Packages/Fortress_SharedMaterials/GlobalMaterialParameters/FortniteMaterialParameters.FortniteMaterialParameters" "SunlightDirectionalVector" (Math|Conversions|ToLinearColor(Vector) (* (Transformation|GetForwardVector (Variables|Default|GetArrow)) -1.0)) :!class /Script/Engine.KismetMaterialLibrary))

(event Collision|EventActorBeginOverlap
  (params (OtherActor :type Actor%20Object%20Reference)))

(event EventTick
  (params (DeltaSeconds :type Float%20%28single-precision%29)))

(fn ConstructionScript
  (Rendering|Material|SetVectorParameterValue "/Game/Packages/Fortress_SharedMaterials/GlobalMaterialParameters/FortniteMaterialParameters.FortniteMaterialParameters" "SunlightDirectionalVector" (Math|Conversions|ToLinearColor(Vector) (* (Transformation|GetForwardVector (Variables|Default|GetArrow)) -1.0)) :!class /Script/Engine.KismetMaterialLibrary))
