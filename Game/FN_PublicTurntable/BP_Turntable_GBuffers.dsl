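; Turntable capture actor that switches between studio lighting and G-buffer passes (base color, roughness, specular, metallic, emissive) with an optional info overlay.
; SetPP toggles per-pass post-process components, updates the UI label and sets anti-aliasing; emissive uses ShowFlag console commands.
; Construction gathers user, date, build and a mesh list with asset-tag vertex counts, and enables inset shadows on turntable objects.
; Tick feeds the capture frame number to the widget.
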(event EventBeginPlay
  (if (Variables|Default|GetDisplayUI)
    ;[1 1 1] setup HUD
    (bind _returnvalue (UserInterface|CreateWidget "/Game/FN_PublicTurntable/WBP_Turntable_UI.WBP_Turntable_UI_C"))
    (bind _output_get (Variables|Default|SetUI _returnvalue))
    (UserInterface|Viewport|AddtoViewport _output_get 0)
    (Class|WBPTurntableUI|SetText (Variables|Default|GetUI) (Variables|Default|GetuserName) (Variables|Default|Getdate) (Variables|Default|Getbuild) (Variables|Default|GetobjectList) "lit" "0001" (Variables|Default|Getversion))))

(event Collision|EventActorBeginOverlap
  (params (OtherActor :type Actor%20Object%20Reference)))

(event EventTick
  (params (DeltaSeconds :type Float%20%28single-precision%29))
  (Class|WBPTurntableUI|SetFrameNumber (Variables|Default|GetUI) (Cinematics|Capture|GetCaptureFrameNumber)))

(event EventEndPlay
  (params (EndPlayReason :type EEndPlayReason%20Enum :default "Destroyed"))
  (Development|ExecuteConsoleCommand "r.TemporalAASamples 8")
  (CallFunction|Lit))

(event Custom|LevelChange
  (params (Pass :type String))
  (CallFunction|SetPP true)
  (Class|WBPTurntableUI|SetPass (Variables|Default|GetUI) Pass))

(fn baseColor
  (CallFunction|SetPP false true))

(fn checkVertexCount
  (for _array_element (Components|GetChildrenComponents (Variables|/Script/Engine.Actor|GetRootComponent (Variables|Default|GetTurntable_Objects)) false)
    (bind _asstatic_mesh_component (Utilities|Casting|CastToStaticMeshComponent _array_element))
    (bind _returnvalue (CallFunction|VertexCount (Variables|/Script/Engine.StaticMeshComponent|GetStaticMesh _asstatic_mesh_component)))
    (Development|PrintString (Utilities|String|Append "staticMesh " (Utilities|GetDisplayName (Variables|/Script/Engine.StaticMeshComponent|GetStaticMesh _asstatic_mesh_component)) " " (Utilities|String|ToString(Integer) _returnvalue)))))

(fn emissive
  (CallFunction|SetPP false false false false false true))

(fn lit
  (CallFunction|SetPP true)
  (CallFunction|SetupObjectList))

(fn Metallic
  (CallFunction|SetPP false false false false true))

(fn roughness
  (CallFunction|SetPP false false true))

(fn setPP
  (params (Lit :type Boolean) (BaseColor :type Boolean) (Roughness :type Boolean) (Specular :type Boolean) (Metallic :type Boolean) (Emissive :type Boolean))
  (bind _ui (Variables|Default|GetUI))
  (Utilities|FlowControl|Sequence
    (:then_0
      ;[1 1 1] PostProcess Enable
      (Variables|/Script/Engine.PostProcessComponent|SetbEnabled BaseColor (Variables|Default|GetPostProcess_BaseColor))
      (Variables|/Script/Engine.PostProcessComponent|SetbEnabled Roughness (Variables|Default|GetPostProcess_Roughness))
      (Variables|/Script/Engine.PostProcessComponent|SetbEnabled Specular (Variables|Default|GetPostProcess_Specular))
      (Variables|/Script/Engine.PostProcessComponent|SetbEnabled Metallic (Variables|Default|GetPostProcess_Metallic)))
    (:then_1
      ;[1 1 1] UI Pass
      (if Lit
        (Class|WBPTurntableUI|SetPass _ui "Studio Lighting")
        (elif BaseColor
          (Class|WBPTurntableUI|SetPass _ui "BaseColor")
          (elif Roughness
            (Class|WBPTurntableUI|SetPass _ui "Roughness")
            (elif Specular
              (Class|WBPTurntableUI|SetPass _ui "Specular")
              (elif Metallic
                (Class|WBPTurntableUI|SetPass _ui "Metallic")
                (elif Emissive
                  (Class|WBPTurntableUI|SetPass _ui "Emissive")
                  (else
                    (Class|WBPTurntableUI|SetPass _ui "Unknown")))))))))
    (:then_2
      ;[1 1 1] AA settings
      (if Lit
        (Development|ExecuteConsoleCommand "r.TemporalAASamples 8")
        (Development|ExecuteConsoleCommand "r.PostProcessAAQuality 6")
        (else
          (Development|ExecuteConsoleCommand "r.TemporalAASamples 1"))))
    (:then_3
      ;[1 1 1] Emissive
      (if Emissive
        (Development|ExecuteConsoleCommand "ShowFlag.Diffuse 0")
        (Development|ExecuteConsoleCommand "ShowFlag.Specular 0")
        (Development|ExecuteConsoleCommand "ShowFlag.Lighting 0")
        (else
          (Development|ExecuteConsoleCommand "ShowFlag.Diffuse 1")
          (Development|ExecuteConsoleCommand "ShowFlag.Specular 1")
          (Development|ExecuteConsoleCommand "ShowFlag.Lighting 1"))))))

(fn setup
  (CallFunction|SetupShadows)
  (CallFunction|SetupObjectList)
  (CallFunction|Lit)
  (Class|WBPTurntableUI|SetText (Variables|Default|GetUI) (Variables|Default|GetuserName) (Variables|Default|Getdate) (Variables|Default|Getbuild) (Variables|Default|GetobjectList) "lit" "0001" (Variables|Default|Getversion))
  (Development|ExecuteConsoleCommand "r.MaterialQualityLevel 1")
  (CallFunction|Lit))

(fn setupLevelsChecker)

(fn setupObjectList
  (bind _objectlist (Variables|Default|GetobjectList))
  (Variables|Default|SetobjectList "")
  (for _array_element (Components|GetChildrenComponents (Variables|/Script/Engine.Actor|GetRootComponent (Variables|Default|GetTurntable_Objects)) false)
    (bind _asstatic_mesh_component (Utilities|Casting|CastToStaticMeshComponent _array_element)
      (:then
        (bind _returnvalue (CallFunction|VertexCount (Variables|/Script/Engine.StaticMeshComponent|GetStaticMesh _asstatic_mesh_component)))
        (Variables|Default|SetobjectList (Utilities|String|Append _objectlist (Utilities|String|Split (Utilities|GetDisplayName _array_element) ".") " (SM " (Utilities|String|ToString(Integer) _returnvalue) ")
")))
      (:CastFailed
        (bind _asskeletal_mesh_component (Utilities|Casting|CastToSkeletalMeshComponent _array_element))
        (bind _returnvalue_2 (CallFunction|VertexCountSkeletalMesh (Variables|/Script/Engine.SkinnedMeshComponent|GetSkeletalMesh _asskeletal_mesh_component)))
        (Variables|Default|SetobjectList (Utilities|String|Append _objectlist (Utilities|String|Split (Utilities|GetDisplayName _array_element) ".") " (SKM " (Utilities|String|ToString(Integer) _returnvalue_2) ")
"))))))

(fn setupShadows
  (Utilities|FlowControl|Sequence
    (:then_0
      (Development|ExecuteConsoleCommand "r.Shadow.PerObjectDirectionalDepthBias 10"))
    (:then_1
      (Development|ExecuteConsoleCommand "r.Shadow.PerObjectDirectionalSlopeDepthBias 10"))
    (:then_2
      (for _array_element (Components|GetChildrenComponents (Variables|/Script/Engine.Actor|GetRootComponent (Variables|Default|GetTurntable_Objects)) false)
        (bind _asprimitive_component (Utilities|Casting|CastToPrimitiveComponent _array_element))
        (Rendering|SetCastInsetShadow _asprimitive_component true)))))

(fn specular
  (CallFunction|SetPP false false false true))

(fn ConstructionScript
  (bind (_lefts _rights) (Utilities|String|Split (Utilities|String|Split (Development|GetEngineVersion) "+++") "-"))
  ;[1 1 1] UserName
  (Variables|Default|SetuserName (Utilities|Platform|GetPlatformUserName))
  ;[1 1 1] Date
  (Variables|Default|Setdate (Utilities|String|ToString(Text) (Utilities|Text|AsDate(fromUTC) (Math|DateTime|UTCNow) "EDT")))
  ;[1 1 1] Build #
  (Variables|Default|Setbuild _rights)
  ;[1 1 1] Object List + vertex count
  (CallFunction|SetupObjectList)
  (CallFunction|SetupShadows))

(fn vertexCount
  (params (staticMesh :type Static%20Mesh%20Object%20Reference))
  (bind _returnvalue (Utilities|Name|MakeLiteralName "Vertices"))
  (bind _returnvalue_2 (AssetData|CreateAssetData staticMesh))
  (bind (_outtagvalue _returnvalue_3) (AssetData|GetTagValue _returnvalue_2 _returnvalue))
  (if _returnvalue_3
    (return (Utilities|String|StringToInteger _outtagvalue))
    (else
      (return -1))))

(fn vertexCount_SkeletalMesh
  (params (skeletalMEsh :type Skeletal%20Mesh%20Object%20Reference))
  (bind _returnvalue (Utilities|Name|MakeLiteralName "Vertices"))
  (bind _returnvalue_2 (AssetData|CreateAssetData skeletalMEsh))
  (bind (_outtagvalue _returnvalue_3) (AssetData|GetTagValue _returnvalue_2 _returnvalue))
  (if _returnvalue_3
    (return (Utilities|String|StringToInteger _outtagvalue))
    (else
      (return -1))))
