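; Base turntable actor for presenting an asset, placed under its "-> ADD ASSET HERE <-" component.
; The Construction Script forces every skeletal and static mesh component to LOD 0, using the one-based forced LOD value of 1.
; It then selects that component's relative transform by Asset Type, either zeroing location and rotation or applying preset height, rotation and scale adjustments.
; BeginPlay, overlap and Tick events are empty.
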(event EventBeginPlay)

(event Collision|EventActorBeginOverlap
  (params (OtherActor :type Actor%20Object%20Reference)))

(event EventTick
  (params (DeltaSeconds :type Float%20%28single-precision%29)))

(fn ConstructionScript
  (Utilities|FlowControl|Sequence
    (:then_0
      (Utilities|FlowControl|Sequence
        (:then_0
          (for _array_element (Actor|GetComponentsByClass "/Script/Engine.SkeletalMeshComponent")
            (bind _asskeletal_mesh_component (Utilities|Casting|CastToSkeletalMeshComponent _array_element))
            (Components|SkinnedMesh|SetForcedLOD _asskeletal_mesh_component 1)))
        (:then_1
          (for _array_element_2 (Actor|GetComponentsByClass "/Script/Engine.StaticMeshComponent")
            (bind _asstatic_mesh_component (Utilities|Casting|CastToStaticMeshComponent _array_element_2))
            (Rendering|LOD|SetForcedLodModel _asstatic_mesh_component 1)))))
    (:then_1
      (Transformation|SetRelativeTransform (Variables|Default|Get->%20ADD%20ASSET%20HERE%20<-) (Utilities|Select :Index (Variables|Default|GetAsset%20Type) :NewEnumerator0 (Math|Transform|MakeTransform "0, 0, 0" "0, 0, 0" (.scale (Transformation|GetRelativeTransform (Variables|Default|Get->%20ADD%20ASSET%20HERE%20<-)))) :NewEnumerator1 (Math|Transform|MakeTransform (+ (.location (Transformation|GetRelativeTransform (Variables|Default|Get->%20ADD%20ASSET%20HERE%20<-))) "0, 0,80.000000") (Math|Rotator|CombineRotators (.rotation (Transformation|GetRelativeTransform (Variables|Default|Get->%20ADD%20ASSET%20HERE%20<-))) "0.000000, 0,33.000000") (* (.scale (Transformation|GetRelativeTransform (Variables|Default|Get->%20ADD%20ASSET%20HERE%20<-))) "0.900000,0.900000,0.900000")) :NewEnumerator2 (Math|Transform|MakeTransform (+ (.location (Transformation|GetRelativeTransform (Variables|Default|Get->%20ADD%20ASSET%20HERE%20<-))) "0,0.000000,80.000000") (.rotation (Transformation|GetRelativeTransform (Variables|Default|Get->%20ADD%20ASSET%20HERE%20<-))) (* (.scale (Transformation|GetRelativeTransform (Variables|Default|Get->%20ADD%20ASSET%20HERE%20<-))) "1.100000,1.100000,1.100000")) :NewEnumerator3 (Math|Transform|MakeTransform (+ (.location (Transformation|GetRelativeTransform (Variables|Default|Get->%20ADD%20ASSET%20HERE%20<-))) "0, 0,100.000000") (.rotation (Transformation|GetRelativeTransform (Variables|Default|Get->%20ADD%20ASSET%20HERE%20<-))) (* (.scale (Transformation|GetRelativeTransform (Variables|Default|Get->%20ADD%20ASSET%20HERE%20<-))) "1.800000,1.800000,1.800000")) :NewEnumerator4 (Math|Transform|MakeTransform (+ (.location (Transformation|GetRelativeTransform (Variables|Default|Get->%20ADD%20ASSET%20HERE%20<-))) "0, 0,90.000000") (.rotation (Transformation|GetRelativeTransform (Variables|Default|Get->%20ADD%20ASSET%20HERE%20<-))) (.scale (Transformation|GetRelativeTransform (Variables|Default|Get->%20ADD%20ASSET%20HERE%20<-))))) false false))))
